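; Zero Point dash gameplay ability that launches the Fortnite pawn toward a computed point on a cooldown.
; Target XY follows movement acceleration, or forward when idle, while height follows aim pitch.
; After commit, it skips clambering, blocking tags, skydiving and synced actions, then runs a curve-driven root motion MoveToForce with custom exit velocity.
; The server applies active and quest effects, a cue plays, falling height resets, and the ability ends after the dash duration.
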(event Custom|WaitForClamberTags
  (bind _t_clambering (Variables|Default|GetT_Clambering))
  ;[1 0 0.815] If the player is in clambering range, stop the montage and update our variable so we can prevent spawning a crashpad.
  (Ability|Tasks|WaitGameplayTagAdd _t_clambering
    (:then
      (Ability|Tasks|WaitGameplayTagRemove _t_clambering
        (:Removed
          (Variables|Default|SetbIsClambering false))))
    (:Added
      (Variables|Default|SetbIsClambering true)
      (Ability|Animation|MontageStop -1.0))))

(event Ability|EventActivateAbility
  (bind _returnvalue (CallFunction|CalculateTargetLocation))
  ;[1 1 1] On Begin, immediately set the camera and commit the cooldown
  (Camera|PushCameraMode "/Game/Athena/Items/ForagedItems/ZeroPoint/ZeroPoint_Dash_CameraNew.ZeroPoint_Dash_CameraNew_C")
  (HUD|SetHUDElementVisibility (BlueprintContext|GetFortHUDContext) (Variables|Default|GetT_HudElementToHide) true "None")
  (bind _ability_committed (CallFunction|AttemptCommitAbility))
  (if _ability_committed
    (bind _asfort_player_pawn (Utilities|Casting|CastToFortPlayerPawn (Ability|GetAvatarActorfromActorInfo))
      (:then
        (bind _output_get (Variables|Default|SetFortPawn _asfort_player_pawn))
        (Utilities|IsValid _output_get
          (:Is%20Valid
            ;[1 1 1] Check for Clamber Tags and stop ability if clambering
            (CallFunction|WaitforClamberTags)
            (if (Variables|Default|GetbIsClambering)
              _
              (elif (GameplayTags|HasAnyMatchingGameplayTags _output_get (GameplayTags|MakeGameplayTagContainerfromArray (Variables|Default|GetTC%20Blocking%20Tags)))
                _
                (else
                  ;[1 1 1] Check if we are skydiving, if we are, end the ability.
                  (Utilities|IsSkydiving (Ability|GetAvatarActorfromActorInfo)
                    (:true
                      _)
                    (:false
                      ;[1 1 1] If we are doing a sync action (hurdle, mantle, etc..), just cancel it
                      (if (not (GameplayTags|HasMatchingGameplayTag (Variables|Default|GetFortPawn) "(TagName=\"MovementMode.SynchedAction\")"))
                        ;[1 1 1] Calculate Dash Movement
                        (Variables|Default|SetLaunch%20Velocity (/ (- _returnvalue (Transformation|GetActorLocation (Variables|Default|GetFortPawn))) (ScalableFloat|GetValueatLevel (Variables|Default|GetDashDuration))))
                        (Ability|Tasks|ApplyRootMotionMoveToForce "None" _returnvalue (ScalableFloat|GetValueatLevel (Variables|Default|GetDashDuration)) true "MOVE_Falling" false "/Game/Athena/Items/ForagedItems/ZeroPoint/Curve_ZeroPoint_Dash.Curve_ZeroPoint_Dash" "SetVelocity" (/ (+ (CallFunction|GetMovementEndVelocity (Variables|Default|GetFortPawn)) (Pawn|Components|CharacterMovement|GetCurrentAcceleration (Variables|/Script/Engine.Character|GetCharacterMovement (Variables|Default|GetFortPawn)))) 1.5)
                          (:then
                            (if (Networking|IsServer)
                              (Ability|ApplyGameplayEffectToOwner "/Game/Athena/Items/ForagedItems/ZeroPoint/GE_Quest_AfterDashing10s.GE_Quest_AfterDashing10s_C")
                              (bind _returnvalue_2 (Ability|ApplyGameplayEffectToOwner "/Game/Athena/Items/ForagedItems/ZeroPoint/GE_ZeroPoint_IsActive.GE_ZeroPoint_IsActive_C"))
                              (Variables|Default|SetGE_IsActive _returnvalue_2)
                              (else
                                _))
                            (Ability|ExecuteGameplayCueOnOwner (Variables|Default|GetGC_Cue))
                            (Pawn|Character|ResetFallingHeight (Variables|Default|GetFortPawn))
                            (Utilities|FlowControl|Delay (+ (ScalableFloat|GetValueatLevel (Variables|Default|GetDashDuration)) 0.1))
                            (Ability|EndAbility))
                          (:OnTimedOutAndDestinationReached
                            (Ability|RemoveGameplayEffectFromOwnerWithHandle (Variables|Default|GetGE_IsActive)))))))))))
          (:Is%20Not%20Valid
            (Utilities|FlowControl|Delay (Variables|Default|GetFailsafe%20Delay))))
        (Ability|EndAbility))
      (:CastFailed
        _))
    (else
      _))
  (Ability|EndAbility))

(event Ability|EventOnEndAbility
  (params (bWasCancelled :type Boolean))
  ;[1 1 1] End Ability
  (HUD|SetHUDElementVisibility (BlueprintContext|GetFortHUDContext) (Variables|Default|GetT_HudElementToHide) true "None")
  (Ability|RemoveGameplayEffectFromOwnerWithHandle (Variables|Default|GetGE_IsActive)))

(fn AttemptCommitAbility
  ;[1 0.8122 0.2487] Handle commit ability in function to allow overriding cost type
  (bind _returnvalue (Ability|CommitAbilityCooldown false))
  (return _returnvalue))

(fn CalculateTargetLocation
  (bind (_returnvalue_x _returnvalue_y _returnvalue_z) (Math|Vector|BreakVector (CallFunction|GetTargetLocationBasedonAimRotation)))
  (bind (_targetlocation_x _targetlocation_y) (Math|Vector|BreakVector (CallFunction|GetTargetLocationBasedonAcceleration)))
  (bind (_returnvalue_x_2 _returnvalue_y_2 _returnvalue_z_2) (Math|Vector|BreakVector (Transformation|GetActorLocation (Variables|Default|GetFortPawn))))
  (return (Math|Vector|MakeVector _targetlocation_x _targetlocation_y (+ _returnvalue_z _returnvalue_z_2))))

(fn GetMovementEndVelocity
  (params (FortPawn :type Fort%20Player%20Pawn%20Object%20Reference))
  (return (* (Math|Vector|Normalize (Math|Vector|MakeVector (.x (Math|Vector|SelectVector (Transformation|GetActorForwardVector FortPawn) (Transformation|GetVelocity FortPawn) (or (== (Math|Vector|VectorLength (Transformation|GetVelocity FortPawn)) 0.0) (and (== (.x (Transformation|GetVelocity FortPawn)) 0.0) (== (.y (Transformation|GetVelocity FortPawn)) 0.0))))) (.y (Math|Vector|SelectVector (Transformation|GetActorForwardVector FortPawn) (Transformation|GetVelocity FortPawn) (or (== (Math|Vector|VectorLength (Transformation|GetVelocity FortPawn)) 0.0) (and (== (.x (Transformation|GetVelocity FortPawn)) 0.0) (== (.y (Transformation|GetVelocity FortPawn)) 0.0))))))) (Variables|Default|GetF_EndVelocityMult))))

(fn GetTargetLocationBasedOnAcceleration
  (return (+ (Transformation|GetActorLocation (Variables|Default|GetFortPawn)) (* (Math|Vector|Normalize (Math|Vector|MakeVector (.x (Math|Vector|SelectVector (Transformation|GetActorForwardVector (Variables|Default|GetFortPawn)) (Pawn|Components|CharacterMovement|GetCurrentAcceleration (Variables|/Script/Engine.Character|GetCharacterMovement (Variables|Default|GetFortPawn))) (or (== (Math|Vector|VectorLength (Pawn|Components|CharacterMovement|GetCurrentAcceleration (Variables|/Script/Engine.Character|GetCharacterMovement (Variables|Default|GetFortPawn)))) 0.0) (and (== (.x (Pawn|Components|CharacterMovement|GetCurrentAcceleration (Variables|/Script/Engine.Character|GetCharacterMovement (Variables|Default|GetFortPawn)))) 0.0) (== (.y (Pawn|Components|CharacterMovement|GetCurrentAcceleration (Variables|/Script/Engine.Character|GetCharacterMovement (Variables|Default|GetFortPawn)))) 0.0))))) (.y (Math|Vector|SelectVector (Transformation|GetActorForwardVector (Variables|Default|GetFortPawn)) (Pawn|Components|CharacterMovement|GetCurrentAcceleration (Variables|/Script/Engine.Character|GetCharacterMovement (Variables|Default|GetFortPawn))) (or (== (Math|Vector|VectorLength (Pawn|Components|CharacterMovement|GetCurrentAcceleration (Variables|/Script/Engine.Character|GetCharacterMovement (Variables|Default|GetFortPawn)))) 0.0) (and (== (.x (Pawn|Components|CharacterMovement|GetCurrentAcceleration (Variables|/Script/Engine.Character|GetCharacterMovement (Variables|Default|GetFortPawn)))) 0.0) (== (.y (Pawn|Components|CharacterMovement|GetCurrentAcceleration (Variables|/Script/Engine.Character|GetCharacterMovement (Variables|Default|GetFortPawn)))) 0.0))))))) (CallFunction|GetMoveDistance)))))

(fn GetTargetLocationBasedOnAimRotation
  (return (* (Math|Vector|Normalize (Math|Rotator|GetRotationXVector (Targeting|GetDefaultTargetingRotation (Variables|Default|GetFortPawn)))) (CallFunction|GetMoveDistance))))

(fn GetMoveDistance
  (return (ScalableFloat|GetValueatLevel (Variables|Default|GetMoveDistance))))

(fn ImpactEnemy)

(fn SetPawnCollision
  (params (Condition :type Boolean))
  (bind _fortpawn (Variables|Default|GetFortPawn))
  (bind _capsulecomponent (Variables|/Script/Engine.Character|GetCapsuleComponent _fortpawn))
  (bind _fortpawn_2 (Variables|Default|GetFortPawn))
  (bind _capsulecomponent_2 (Variables|/Script/Engine.Character|GetCapsuleComponent _fortpawn_2))
  (bind _fortpawn_3 (Variables|Default|GetFortPawn))
  (bind _capsulecomponent_3 (Variables|/Script/Engine.Character|GetCapsuleComponent _fortpawn_3))
  (if Condition
    (if (Utilities|Name|NotEqual(Name) (Variables|Default|GetCollision%20Profile) "None")
      (Collision|SetCollisionProfileName _capsulecomponent_2 (Variables|Default|GetCollision%20Profile) true :!class /Script/Engine.PrimitiveComponent)
      (Collision|SetCollisionResponsetoChannel _capsulecomponent_3 "ECC_Pawn" "ECR_Block")
      (Collision|SetCollisionResponsetoChannel _capsulecomponent_3 "ECC_GameTraceChannel4" "ECR_Block"))
    (else
      (Variables|Default|SetCollision%20Profile (Collision|GetCollisionProfileName _capsulecomponent_2))
      (Collision|SetCollisionResponsetoChannel _capsulecomponent "ECC_Pawn" "ECR_Overlap")
      (Collision|SetCollisionResponsetoChannel _capsulecomponent "ECC_GameTraceChannel4" "ECR_Overlap"))))
